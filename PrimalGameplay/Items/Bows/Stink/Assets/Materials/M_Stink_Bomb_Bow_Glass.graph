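# Unlit glass material for the Stink Bomb bow, usable on skeletal meshes, showing animated smoke with glowing highlights.
# Two panning cloud noise layers and matching normals blend to drive smoke color, camera-facing glow and a flattened-normal Fresnel mask.
# The color fades along texture V and blends into a glass edge color by squared Fresnel.
# Shared LOD pickup and black hole functions compute a combined vertex-style offset.
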Material /PrimalGameplay/Items/Bows/Stink/Assets/Materials/M_Stink_Bomb_Bow_Glass
graph {
 properties {
  allowTranslucentCustomDepthWrites=true
  bUsedWithMeshDeformer=true
  bUsedWithSkeletalMesh=true
  pixelDepthOffsetMode=PDOM_Legacy
  shadingModel=MSM_Unlit
  usedShadingModels=Unlit
 }
 nodes [
  n1 = VectorParameter() {parameterName=WorldPostionOfBlackHole, defaultValue=<rgba -30000 25536 1952 0>}
  n2 = MaterialFunctionCall!(
   World Position
   DistanceCullFadeAnimationOffset=MaterialFunctionCall() {materialFunction=/Game/Packages/Fortress_SharedMaterials/Base_Material_Functions/PickupEffectDistanceCullOverride}.Result
   Use actor center instead of pivot point
  ) {materialFunction=/Game/Packages/Fortress_SharedMaterials/Base_Material_Functions/LODEffect}
  n3 = WorldPosition()
  Add(
   A=MaterialFunctionCall!(WorldPostionOfBlackHole=AppendVector(A=@n1.RGB, B=@n1.RGB), World Position=Add(A=@n2.Result, B=@n3.XYZ)) {materialFunction=/Game/Packages/Fortnite_UniqueMaterials_1/BlackHoleFunction}.Result
   B=@n2.Result
  )
  n4 = Time()
  n5 = Panner(Coordinate=TextureCoordinate(), Time=@n4) {speedX=0.075, speedY=0.105}
  n6 = Panner(Coordinate=TextureCoordinate(), Time=@n4) {speedX=-0.015, speedY=0.09}
  n7 = LinearInterpolate(
   A=TextureSample(UVs=@n5) {samplerType=SAMPLERTYPE_LinearGrayscale, texture=/Game/Environments/Hexmap/HexClouds/T_HexmapFog_01_Height_GS}.R
   B=TextureSample(UVs=@n6) {samplerType=SAMPLERTYPE_LinearGrayscale, texture=/Game/Global/Textures/Default/Tileable_Patterns/T_StormCloudDetail01_M}.R
  ) {constAlpha=0.4}
  Clamp(None=LinearInterpolate(Alpha=@n7) {constA=-0.1})
  n8 = VectorParameter() {parameterName=Color2, defaultValue=<rgba 0.3229 0.2534 0.1827 0>}
  n9 = VectorParameter() {parameterName=SmokeColor2, defaultValue=<rgba 0.0354 0.04 0.0163 0>}
  n10 = VectorParameter() {parameterName=Color1, defaultValue=<rgba 2 1.1212 0.2812 0>}
  n11 = LinearInterpolate(
   A=TextureSample(UVs=@n5) {samplerType=SAMPLERTYPE_Normal, texture=/Game/Global/Textures/Default/Tileable_Patterns/T_StormCloudDetail01_N}.RGB
   B=TextureSample(UVs=@n6) {samplerType=SAMPLERTYPE_Normal, texture=&T_StormCloudDetail01_N}.RGB
  )
  n12 = Power(Base=DotProduct(A=Transform(None=@n11), B=CameraVectorWS())) {constExponent=4}
  n13 = Add(A=Add(A=LinearInterpolate(A=@n8.RGB, B=@n9.RGB, Alpha=@n7), B=Multiply(A=@n8.RGB)), B=Multiply(A=@n10.RGB, B=Add(A=@n12) {constB=0.05}))
  n14 = LinearInterpolate(
   A=ScalarParameter() {defaultValue=1, parameterName=EdgeGlowMax}
   B=ScalarParameter() {defaultValue=-0.5, parameterName=EdgeGlowMin}
   Alpha=@n7
  ) {constA=1, constB=-0.3}
  n15 = Multiply(A=@n10.RGB, B=Multiply(A=Clamp(None=@n14), B=@n12))
  n16 = TextureCoordinate()
  n17 = Fresnel() {baseReflectFraction=0, exponent=0.85}
  LinearInterpolate(
   A=Multiply(A=Add(A=@n13, B=@n15), B=ComponentMask(None=@n16) {g=true})
   B=VectorParameter() {parameterName=GlassEdgeColor, defaultValue=<rgba 0.159 0.1069 0 1>}.RGB
   Alpha=Multiply(A=@n17, B=@n17)
  )
  n18 = MaterialFunctionCall(Normal=@n11, Flatness=Constant() {r=0.7}) {materialFunction=&FlattenNormal}
  n19 = LinearInterpolate(Alpha=Fresnel(Normal=Transform(None=@n18.Result)) {baseReflectFraction=0, exponent=0.1}) {constA=2, constB=-0.25}
  Multiply(A=Multiply(A=Clamp(None=@n19), B=@n7)) {constB=15}
 ]
}
